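# Physics asset for the M_MED_Sleek male player body's dynamic necklace, built entirely from capsule bodies.
# A master necklace body hangs off spine_05 and branches into a two-link chain (01 to 02) and a separate inner segment.
# Necklace links have tight twist and first-swing limits with wider second swing, favoring motion on one axis.
# Upper spine and clavicle capsules approximate the torso, joined to one another by free angular constraints.
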PhysicsAsset /BRCosmetics/Characters/Player/Male/Medium/Bodies/M_MED_Sleek/Meshes/M_MED_Sleek_dyn_necklace
bodies {
 clavicle_l [{
   shapeType=Capsule
   length=8.5904
   radius=7.0039
   shapeName=clavicle_l_capsule
   center=<11.3085 -0.2 -3.4318>
   rotation=<rot 90 19.7672 17.8373>
  }]
 clavicle_r [{shapeType=Capsule, length=8.5904, radius=7.0039, shapeName=clavicle_l_capsule, center=<-11.3085 0.2 3.4318>, rotation=<rot 90 -90 88.0701>}]
 dyn_necklace_01 [{shapeType=Capsule, length=4.8278, radius=2.4222, center=<3 0 2.5>, rotation=<rot 90 -0.0198 -0.0198>}]
 dyn_necklace_02 [{shapeType=Capsule, length=4.5, radius=3.3502, center=<3 0 2.5>, rotation=<rot 87.28 -180 -180>}]
 dyn_necklace_MSTR [{shapeType=Capsule, length=0.5, radius=0.5}]
 dyn_necklace_inner [{shapeType=Capsule, length=4.8278, radius=2.4222, center=<3 0 2.5>, rotation=<rot 90 -0.0198 -0.0198>}]
 spine_04 [{shapeType=Capsule, length=15.2009, radius=12.4611, center=<3.5142 0.1594 0.0001>, rotation=<rot 0 89.9999 0>}]
 spine_05 [{shapeType=Capsule, length=19.0759, radius=11.9611, center=<-1.7112 -1.5237 0.0001>, rotation=<rot 0 89.9999 0>}]
}
constraints [
 {
  bone1Name=dyn_necklace_MSTR
  bone2Name=spine_05
  swing1LimitDegrees=45
  swing1Motion=Free
  swing2LimitDegrees=45
  swing2Motion=Free
  twistLimitDegrees=45
  twistMotion=Free
 }
 {bone1Name=dyn_necklace_01, bone2Name=dyn_necklace_MSTR, swing1LimitDegrees=8, swing2LimitDegrees=30, twistLimitDegrees=2}
 {bone1Name=dyn_necklace_02, bone2Name=dyn_necklace_01, swing1LimitDegrees=4, swing2LimitDegrees=30, twistLimitDegrees=2}
 {
  bone1Name=spine_05
  bone2Name=spine_04
  swing1LimitDegrees=45
  swing1Motion=Free
  swing2LimitDegrees=45
  swing2Motion=Free
  twistLimitDegrees=45
  twistMotion=Free
 }
 {
  bone1Name=clavicle_l
  bone2Name=spine_05
  swing1LimitDegrees=45
  swing1Motion=Free
  swing2LimitDegrees=45
  swing2Motion=Free
  twistLimitDegrees=45
  twistMotion=Free
 }
 {bone1Name=dyn_necklace_inner, bone2Name=dyn_necklace_MSTR, swing1LimitDegrees=10, swing2LimitDegrees=30, twistLimitDegrees=2}
 {
  bone1Name=clavicle_r
  bone2Name=spine_05
  swing1LimitDegrees=45
  swing1Motion=Free
  swing2LimitDegrees=45
  swing2Motion=Free
  twistLimitDegrees=45
  twistMotion=Free
 }
]
