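# Post-pose animation blueprint for the M_MED_Sleek player body, adding deformation correctives and necklace motion.
# It copies the attached parent's pose, runs twist and sculpter Control Rigs, then offsets the right triceps bone by upper-arm yaw.
# The cached result feeds a necklace RigidBody simulation or two Look At nodes, selected by bRigidBodyEnabled.
# The event graph computes both arm-rotation values each update, but the AnimGraph uses only the right.
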AnimBlueprint /BRCosmetics/Characters/Player/Male/Medium/Bodies/M_MED_Sleek/Meshes/M_MED_Sleek_AnimBP
target_skeleton=&/M_MED_Sleek_Skeleton
graphs [
 AnimGraph {nodes [
   n1 = RigidBody!(ComponentPose=Local To Component(LocalPose=Use cached pose 'prePose'()), Alpha, OverrideWorldGravity:Vector="0,0,-980", ExternalForce) -> (Pose) {
    bClampLinearTranslationLimitToRefPose=true
    bOverrideWorldGravity=true
    overridePhysicsAsset=&/M_MED_Sleek_dyn_necklace
    simulationSpace=BaseBoneSpace
    alphaBoolBlend {blendOption=Linear}
    baseBoneRef {boneName=spine_05}
    componentAppliedLinearAccClamp=<2000 2000 1000>
    componentLinearAccScale {x=0.4}
    componentLinearVelScale {x=0.4}
    overrideWorldGravity {z=-980}
    simSpaceSettings {maxAngularAcceleration=100, maxAngularVelocity=5, maxLinearAcceleration=300, maxLinearVelocity=50, velocityScaleZ=0, worldAlpha=1}
   }
   n2 = Look At - Bone: dyn_necklace_02(ComponentPose=Local To Component(LocalPose=Use cached pose 'prePose'()), Alpha=1) {
    alphaBoolBlend {blendOption=Linear}
    boneToModify {boneName=dyn_necklace_02}
    lookAtLocation=<-30 36>
    lookAtTarget {boneReference {boneName=spine_04}}
    lookAt_Axis {axis=<1 0>}
    lookUp_Axis {axis {x=0}}
   }
   n3 = Look At - Bone: dyn_necklace_03(ComponentPose=@n2, Alpha=1) {
    alphaBoolBlend {blendOption=Linear}
    boneToModify {boneName=dyn_necklace_03}
    lookAtLocation=<-30 23>
    lookAtTarget {boneReference {boneName=spine_03}}
    lookAt_Axis {axis=<1 0>}
    lookUp_Axis {axis {x=0}}
   }
   Output Pose(Result=Blend Poses by bool(BlendPose_0=Component To Local(ComponentPose=@n1), BlendPose_1=Component To Local(ComponentPose=@n3), bActiveValue=Get bRigidBodyEnabled()) {blendType=Linear}) {layerGroup=None}
   n4 = Control Rig(Source=Copy Pose From Mesh() {bUseAttachedParent=true}) {
    alphaBoolBlend {blendOption=Linear}
    controlRigAssetReference {blueprintRigClass=&/M_MED_Sleek_Deform_CtrlRig_twists.M_MED_Sleek_Deform_CtrlRig_twists_C}
   }
   n5 = Control Rig!(
    Source=@n4
    Alpha
    KneeRot_fr_l_114deg:Vector
    KneeTrans_fr_l_114deg:Vector="0.258,-3.591,0.009"
    KneeScale_fr_l_114deg:Vector="1,1,1"
    KneeScale_fr_l_130deg:Vector="1,1,1"
    KneeRot_fr_l_130deg:Vector="0,0,-51.2"
    KneeTrans_fr_l_130deg:Vector="0.258,-3.591,0.009"
   ) -> (Pose) {
    alphaBoolBlend {blendOption=Linear}
    controlRigAssetReference {blueprintRigClass=&/M_MED_Sleek_Deform_CtrlRig_sculpters.M_MED_Sleek_Deform_CtrlRig_sculpters_C}
   }
   n6 = "Transform (Modify) Bone - Bone: deform_triceps_fr_r"(
    ComponentPose=Local To Component(LocalPose=@n5)
    Alpha=Get arm_rot_up_r() -> (arm_rot_up_r:double)
    Translation="0,6,0.000"
    Rotation="0,0,0"
    Scale="1,1,1"
   ) {
    translationMode=BMM_Additive
    translationSpace=BCS_BoneSpace
    alphaBoolBlend {blendOption=Linear}
    boneToModify {boneName=deform_triceps_fr_r}
    translation {y=6}
   }
   Save cached pose 'prePose'(Pose=Component To Local(ComponentPose=@n6)) {cacheName=prePose}
  ]}
 EventGraph {nodes [
   Event BlueprintUpdateAnimation()
   n1 = GetOwnerSkeletalMeshComponent()
   n2 = BreakRotator(InRot=BreakTransform(InTransform=GetDeltaTransformFromRefPose(self=@n1, BoneName=upperarm_r, BaseName=spine_05)).Rotation)
   > Set arm_rot_up_r(arm_rot_up_r:double=MapRangeClamped(Value="Absolute (Float)"(A=@n2.Yaw), InRangeA=20, InRangeB=100)) -> (Output_Get:double)
   n3 = GetOwnerSkeletalMeshComponent()
   n4 = BreakRotator(InRot=BreakTransform(InTransform=GetDeltaTransformFromRefPose(self=@n3, BoneName=upperarm_l, BaseName=spine_05)).Rotation)
   > Set arm_rot_up_l(arm_rot_up_l:double=MapRangeClamped(Value="Absolute (Float)"(A=@n4.Yaw), InRangeA=20, InRangeB=120)) -> (Output_Get:double)
  ]}
]
variables [
 arm_rot_up_l {type=number}
 arm_rot_up_r {type=number}
]
