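# Animation Blueprint for the M_MED_Sleek face accessory that tilts the glasses bone (dyn_galsses) as the cheeks rise.
# Each update, it measures the inner cheek bones' Z delta from ref pose relative to faceAttach and maps it to clamped left and right weights.
# The AnimGraph copies pose and curves from the attached parent mesh, then two chained Modify Bone nodes use those weights as alphas for small additive pitch and opposite roll.
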AnimBlueprint /BRCosmetics/Characters/Player/Male/Medium/Bodies/M_MED_Sleek/Meshes/Parts/M_MED_Sleek_FaceAcc_AnimBP
target_skeleton=&/M_MED_Sleek_FaceAcc_Skeleton
graphs [
 AnimGraph {nodes [
   n1 = "Transform (Modify) Bone - Bone: dyn_galsses"(
    ComponentPose=Local To Component(LocalPose=Copy Pose From Mesh() {bCopyCurves=true, bUseAttachedParent=true})
    Alpha=Get Cheek Up L() -> (Cheek Up L:double)
    Translation="0,0,0"
    Rotation="0.5,0,1"
    Scale="1,1,1"
   ) {
    rotationMode=BMM_Additive
    rotationSpace=BCS_BoneSpace
    translationMode=BMM_Additive
    translationSpace=BCS_BoneSpace
    alphaBoolBlend {blendOption=Linear}
    boneToModify {boneName=dyn_galsses}
    rotation {pitch=0.5, roll=1}
   }
   n2 = "Transform (Modify) Bone - Bone: dyn_galsses"(ComponentPose=@n1, Alpha=Get Cheek Up R() -> (Cheek Up R:double), Translation="0,0,0", Rotation="0.5,0,-1", Scale="1,1,1") {
    rotationMode=BMM_Additive
    rotationSpace=BCS_BoneSpace
    translationMode=BMM_Additive
    translationSpace=BCS_BoneSpace
    alphaBoolBlend {blendOption=Linear}
    boneToModify {boneName=dyn_galsses}
    rotation {pitch=0.5, roll=-1}
   }
   Output Pose(Result=Component To Local(ComponentPose=@n2)) {layerGroup=None}
  ]}
 EventGraph {nodes [
   Event BlueprintUpdateAnimation()
   n1 = BreakTransform(InTransform=GetDeltaTransformFromRefPose(self=GetPartSkeletalMeshComponent(), BoneName=L_cheek_inner, BaseName=faceAttach))
   > Set Cheek Up L(Cheek Up L:double=MapRangeClamped(Value=BreakVector(InVec=@n1).Z, InRangeA=0.1, InRangeB=0.6)) -> (Output_Get:double)
   n2 = BreakTransform(InTransform=GetDeltaTransformFromRefPose(self=GetPartSkeletalMeshComponent(), BoneName=R_cheek_inner, BaseName=faceAttach))
   > Set Cheek Up R(Cheek Up R:double=MapRangeClamped(Value=BreakVector(InVec=@n2).Z, InRangeA=0.1, InRangeB=0.6)) -> (Output_Get:double)
  ]}
]
variables [
 Cheek Up L {type=number}
 Cheek Up R {type=number}
]
